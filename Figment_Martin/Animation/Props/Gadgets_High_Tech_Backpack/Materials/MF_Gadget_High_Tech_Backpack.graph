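# Material function that adds a glowing, segmented charge bar to the high tech backpack, behind the HighTechBackpack static switch.
# ChargeBarScalar snaps to thirteenths and fills a UV window placed by bias and scale parameters.
# Inside that window, the masked emissive texture, scanline gradient and MF_Fresnel_FX replace incoming emissive.
# A sharpened sine pulse boosts the resulting emissive, and the switch's off state passes attributes through unchanged.
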MaterialFunction /Figment_Martin/Animation/Props/Gadgets_High_Tech_Backpack/Materials/MF_Gadget_High_Tech_Backpack
graph {nodes [
  ConstantBiasScale(None=Sine(None=Time()) {period=8})
  n1 = FunctionInput() {inputType=FunctionInput_MaterialAttributes}
  n2 = GetMaterialAttributes(None=@n1) {attributeGetTypes [#1, #2]}
  n3 = TextureSampleParameter2D() {
   group=High Tech Backpack
   parameterName=HighTechEmissiveTexture
   samplerType=SAMPLERTYPE_Masks
   texture=/Figment_Martin/Animation/Props/Gadgets_High_Tech_Backpack/Textures/T_High_Tech_Backpack_E
  }
  n4 = ScalarParameter() {defaultValue=20, group=High Tech Backpack, parameterName=EmissiveTextureScale}
  n5 = Add(
   A=TextureCoordinate()
   B=AppendVector(
    A=ScalarParameter() {defaultValue=-0.45, group=High Tech Backpack, parameterName=UBias}
    B=ScalarParameter() {defaultValue=0.1115, group=High Tech Backpack, parameterName=VBias}
   )
  )
  n6 = AppendVector(
   A=ScalarParameter() {defaultValue=0.1, group=High Tech Backpack, parameterName=UScale}
   B=ScalarParameter() {defaultValue=0.34, group=High Tech Backpack, parameterName=VScale}
  )
  n7 = MaterialFunctionCall(Float2=Saturate(None=MaterialFunctionCall(UVs=@n5, Texture Scale=@n6) {materialFunction=&ScaleUVsByCenter}.UVs)) {materialFunction=&BreakOutFloat2Components}
  n8 = Constant() {r=1}
  n9 = Constant() {r=13}
  n10 = Round(None=Multiply(A=ScalarParameter() {group=High Tech Backpack, parameterName=ChargeBarScalar}, B=@n9))
  n11 = Multiply(A=Multiply(A=Multiply(A=@n3.RGB, B=@n4), B=NamedRerouteUsage()), B=SphereMask(A=@n7.G, B=@n8, Radius=Divide(A=@n10, B=@n9), Hardness=@n8))
  n12 = MaterialFunctionCall(Float2=Multiply(A=TextureCoordinate(), B=ScalarParameter() {defaultValue=150, group=High Tech Backpack, parameterName=ScanlineRepeat}) {constB=100}) {materialFunction=&BreakOutFloat2Components}
  n13 = Constant() {r=0.5}
  n14 = LinearInterpolate(
   A=ScalarParameter() {defaultValue=0.15, group=High Tech Backpack, parameterName=GradientLow}
   B=ScalarParameter() {defaultValue=1, group=High Tech Backpack, parameterName=GradientHigh}
   Alpha=SphereMask(A=Frac(None=@n12.R), B=@n13, Radius=@n13, Hardness=Constant())
  )
  n15 = Constant() {r=0.5}
  n16 = Constant() {r=1}
  n17 = LinearInterpolate(
   A=@n2.EmissiveColor
   B=Multiply(
    A=Multiply(A=@n11, B=@n14)
    B=MaterialFunctionCall() {materialFunction=&MF_Fresnel_FX}.Fresnel
   )
   Alpha=Min(A=SphereMask(A=@n7.R, B=@n15, Radius=@n15, Hardness=@n16), B=SphereMask(A=@n7.G, B=@n15, Radius=@n15, Hardness=@n16))
  )
  n18 = ScalarParameter() {defaultValue=2, group=High Tech Backpack, parameterName=EmissivePulseScale}
  n19 = Time()
  n20 = Power(Base=Abs(None=Sine(None=@n19) {period=1.8})) {constExponent=6}
  n21 = Add(A=@n17, B=Multiply(A=@n17, B=Multiply(A=@n18, B=@n20)))
  FunctionOutput(None=StaticSwitchParameter(
    True=SetMaterialAttributes!(MaterialAttributes=@n2.MaterialAttributes, EmissiveColor=@n21) {attributeSetTypes [#1]}
    False=@n1
   ) {group=Skin FX, parameterName=HighTechBackpack})
  NamedRerouteDeclaration(None=@n2.BaseColor) {name=BaseColorIn, nodeColor=<rgba 0.2422 0 1 1>}
 ]}
